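# Niagara front-end (FE) driving effect for the KartRocket emote breakout, throwing debris, smoke and ribbon trails from kart sockets.
# Debris and Smoke sprites spawn continuously at offsets from sockets read through a shared "NewsVan" socket reader, then fly backward and upward.
# Smoke drag follows owner local velocity and its tint follows time of day.
# Red front and back ribbons use delayed loops; only Trail_Front keeps an enabled speed-gated, front-end-aware spawn rate.
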NiagaraSystem /BRCosmetics/Animation/Game/MainPlayer/Emotes/KartRocket/VFX/NS_KartRocket_Driving_Breakout_FE
dependencies {
 usedDataInterfaces [&NiagaraDataInterfaceCurve, &NiagaraDataInterfaceColorCurve, &NiagaraDataInterfaceSocketReader]
 usedDynamicInputs [&SimulationPosition]
 usedModules [
  &SystemState
  &BlendSpaceHelper
  &self:SystemSpawnScript.SystemScriptSource.SystemScriptGraph.NiagaraNodeAssignment_0.SetVariables_#1
 ]
}
emitters {
 Debris {
  emitterUpdateScript [
   &EmitterState()
   &SpawnRate(SpawnGroup=1, SpawnRate=20, Use Spawn Probability=false) {module=SpawnRate003}
  ]
  particleSpawnScript [
   &InitializeParticle(
    Color=&ColorFromCurve(
     ColorCurve=DataInterface() {propertyValues {
       AlphaCurve {keys [<at -0.0023 1 RCIM_Cubic>]}
       BlueCurve {keys [<at -0.0023 0.2122 RCIM_Cubic>,<at 0.1383 0.2122>,<at 0.2178 0.1859>,<at 0.4486 0.75>,<at 0.7126 0.1859>,<at 1.0234 0.8802>]}
       GreenCurve {keys [<at -0.0023 0.2325 RCIM_Cubic>,<at 0.1383 0.2325>,<at 0.2178 0.2005>,<at 0.4486 0.8456>,<at 0.7126 0.2122>,<at 1.0234 0.9474>]}
       RedCurve {keys [<at -0.0023 0.25 RCIM_Cubic>,<at 0.1383 0.25>,<at 0.2178 0.2188>,<at 0.4486 1>,<at 0.7126 0.2188>,<at 1.0234 1>]}
      }}
     CurveIndex=&UniformRangedFloat(Minimum=0, Maximum=1)
    )
    Lifetime=&UniformRangedFloat(Minimum=0.25, Maximum=0.75)
    Sprite Rotation Mode=::NewEnumerator2
    Sprite Size Mode=::NewEnumerator3
    Uniform Sprite Size=&MakeFloatFromVector2d(Vector 2D=&Vector2DFromFloat(Value=&UniformRangedFloat(Minimum=2.8, Maximum=7)))
    UsePositionOffset=false
    Write Material Random=true
   ) {module=InitializeParticle001}
   &SocketLocation(
    SamplePosOffset=true
    Sampled Position Offset=<0 20 0>
    Socket Index=&MaskIntBySpawnGroup(Masked Int=0, Passthrough Int=1, Spawn Group Mask=0)
    Socket Reader=$System.NewsVan Socket Reader
    Socket Sampling Mode=::NewEnumerator3
    VelocitySpeedLimit=false
   )
   &ShapeLocation(MaskSpawnGroup=false, Override Seed=false, Sphere Radius=2, UseOwnerScale=true)
   &AddVelocity(
    Module.Use Falloff Distance=false
    Origin Offset=<0 -23 0>
    Override Seed=false
    Velocity Mode=::NewEnumerator1
    Velocity Origin=$Output.SocketLocation.SampledPosition
    Velocity Speed=&RandomRangeFloat(Minimum=200, Maximum=300)
   )
   &AddVelocity(Velocity=<0 -0.8 0.3>, Velocity Speed Scale=&RandomRangeFloat(Minimum=300, Maximum=500)) {module=AddVelocity001}
  ]
  particleUpdateScript [
   &ParticleState()
   &Drag(Use Linear Drag=true, Use Rotational Drag=false)
   &GravityForce(Gravity=<0 0 -500>)
   &ScaleColor(Scale RGB=<1.2 1.2 1.2>, ScaleA=true, ScaleRGB=true)
   &ScaleSpriteSize(Scale Factor=&Vector2DFromFloat(Value=&FloatFromCurve(FloatCurve=DataInterface() {propertyValues {Curve {keys [
          <at 0.7 0.9701>
          {arriveTangent=-3.2149, interpMode=RCIM_Cubic, leaveTangent=-3.2149, tangentMode=RCTM_User, time=0.8778, value=0.9191}
          <at 1.0011 0.002>
         ]}}})))
   &SpriteRotationRate()
   &BlendSpace(Rotation Contribution=0.5)
   &SolveForcesAndVelocity(Clamp Velocity=false, Limit Acceleration=false)
   &ScaleSpriteSizeBySpeed(
    Max Scale Factor=&Vector2DFromFloat(Value=1)
    Min Scale Factor=&Vector2DFromFloat(Value=0.25)
    Velocity Threshold=200
   )
  ]
  rendererClasses [&NiagaraSpriteRendererProperties]
  renderers [{rendererClass=&NiagaraSpriteRendererProperties}]
 }
 Smoke {
  emitterUpdateScript [
   &EmitterState()
   &SpawnRate(SpawnGroup=1, SpawnRate=50, Use Spawn Probability=false)
  ]
  particleSpawnScript [
   &InitializeParticle(
    Color Maximum=<rgba 0.781 0.7271 0.5498 0.4>
    Color Minimum=<rgba 0.478 0.3791 0.2199 0.2>
    Color Mode=::NewEnumerator2
    Lifetime Max=0.5
    Lifetime Min=0.3
    Lifetime Mode=::NewEnumerator1
    Sprite Size Mode=::NewEnumerator1
    Uniform Sprite Size Max=50
    Uniform Sprite Size Min=30
    UsePositionOffset=false
    Write Material Random=true
   )
   &SocketLocation(
    SamplePosOffset=true
    Sampled Position Offset=<0 29 0>
    Socket Index=&MaskIntBySpawnGroup(Masked Int=0, Passthrough Int=1, Spawn Group Mask=0)
    Socket Reader=$System.NewsVan Socket Reader
    Socket Sampling Mode=::NewEnumerator3
    VelocitySpeedLimit=false
   )
   &ShapeLocation(MaskSpawnGroup=false, Override Seed=false, Sphere Radius=5, UseOwnerScale=true)
   &AddVelocity(Velocity=<0 -1 0.2>, Velocity Speed Scale=&RandomRangeFloat(Minimum=300, Maximum=500)) {module=AddVelocity001}
  ]
  particleUpdateScript [
   &ParticleState()
   &DynamicMaterialParameters(
    Index 0 Param 1=&Add_Float(
     A=&RandomRangeFloat(Minimum=0, Maximum=1)
     B=&FloatFromCurve(
      FloatCurve=DataInterface() {propertyValues {Curve {keys [
          <at 0 0 RCIM_Cubic>
          {arriveTangent=1.5947, interpMode=RCIM_Cubic, leaveTangent=1.5947, tangentMode=RCTM_User, time=0.2557, value=0.5257}
          <at 1 1 RCIM_Cubic>
         ]}}}
      CurveIndex=$Particles.Age
      Scale Curve=0.25
     )
    )
    Module.Index0Float0Write=true
    Module.Index0Float1Write=true
    Module.Index0Float2Write=true
    Module.Index0Float3Write=true
   ) {module=DynamicMaterialParameters001}
   &Drag(
    Drag=&FloatFromCurve(
     FloatCurve=DataInterface() {propertyValues {Curve {
        defaultValue=1
        keys [{interpMode=RCIM_Cubic, tangentMode=RCTM_User, time=0, value=0.75}, {interpMode=RCIM_Cubic, tangentMode=RCTM_User, time=0.65, value=1}]
       }}}
     Scale Curve=&Lerp_Float(
      A=20
      B=8
      Alpha=&RemapRange(
       Input Value To Remap=&VectorLength(Vector=$User.OwnerLocalVelocity)
       Input Min=0
       Input Max=300
       Output Min=0
       Output Max=1
      )
     )
    )
    Use Linear Drag=true
    Use Rotational Drag=true
   )
   &ScaleSpriteSize(Uniform Curve Sprite Scale=DataInterface() {propertyValues {Curve {keys [
        {arriveTangent=0.8753, interpMode=RCIM_Cubic, leaveTangent=0.8753, tangentMode=RCTM_User, time=0, value=1}
        {arriveTangent=0.0238, interpMode=RCIM_Cubic, leaveTangent=0.0238, tangentMode=RCTM_User, time=1, value=1.3}
       ]}}})
   &ScaleColor(
    Scale Alpha=&FloatFromCurve(
     FloatCurve=DataInterface() {propertyValues {Curve {keys [
         {arriveTangent=0.0878, interpMode=RCIM_Cubic, leaveTangent=0.0878, tangentMode=RCTM_User, time=0, value=0}
         {
          arriveTangent=0.0878
          arriveTangentWeight=0.0754
          interpMode=RCIM_Cubic
          leaveTangent=0.0878
          leaveTangentWeight=0.2297
          tangentMode=RCTM_User
          time=0.1684
          value=1
         }
         {arriveTangent=-0.3893, interpMode=RCIM_Cubic, leaveTangent=-3.5553, tangentMode=RCTM_Break, time=0.4045, value=1}
         {arriveTangent=-0.0368, interpMode=RCIM_Cubic, leaveTangent=-2.9749, tangentMode=RCTM_Break, time=0.9, value=0}
        ]}}}
     Scale Curve=&FN_FrontEndFloatModifier(Gameplay=0.4, Front End=0.4)
    )
    Scale RGB=&MakeVectorFromLinearColorRGB(Color=&FNTimeOfDayColor(
      Night Color=&RandomRangeLinearColor(Minimum=<rgba 0.4138 0.5201 0.6661 1>, Maximum=<rgba 0.781 0.781 0.781 1>)
      Morning Color=&RandomRangeLinearColor(Minimum=<rgba 0.6882 0.6227 0.9531 1>, Maximum=<rgba 0.6063 0.7457 0.9346 1>)
      Day Color=&RandomRangeLinearColor(Minimum=<rgba 0.478 0.4265 0.4121 0.8>, Maximum=<rgba 0.781 0.718 0.6943 0.9>)
      Evening Color=&RandomRangeLinearColor(Minimum=<rgba 0.478 0.478 0.478 1>, Maximum=<rgba 0.781 0.781 0.781 1>)
     ))
    ScaleA=true
    ScaleRGB=true
   )
   &BlendSpace(Rotation Contribution=0.5)
   &SolveForcesAndVelocity(Clamp Velocity=false, Limit Acceleration=false)
  ]
  rendererClasses [&NiagaraSpriteRendererProperties]
  renderers [{rendererClass=&NiagaraSpriteRendererProperties}]
 }
 Trail_Back {
  emitterUpdateScript [
   &EmitterState(Life Cycle Mode=::NewEnumerator1, Loop Behavior=::NewEnumerator1, Loop Delay=0.389, Loop Duration Mode=::NewEnumerator1, UseLoopDelay=true)
   &SpawnRate(SpawnRate=50, Use Spawn Probability=false)
   &SpawnRate(
    SpawnRate=&FN_FrontEndFloatModifier(
     Gameplay=&MakeCustomFloatFromBool(
      Bool=&SetBoolByFloatComparison(A=&VectorLength(Vector=$User.OwnerLocalVelocity), B=10)
      True Float=50
      False Float=0
     )
     Front End=60
    )
    Use Spawn Probability=false
   ) {enabled=false, module=SpawnRate001}
  ]
  particleSpawnScript [
   &InitializeParticle(
    Color=<rgba 1 0 0 0.5>
    Lifetime=0.2
    Ribbon Width=20
    Ribbon Width Mode=::NewEnumerator1
    UsePositionOffset=false
    Write Material Random=true
   )
   &SocketLocation(
    SamplePosOffset=true
    Sampled Position Offset=<3 22 0>
    Socket Index=3
    Socket Reader=$System.NewsVan Socket Reader
    Socket Sampling Mode=::NewEnumerator3
    VelocitySpeedLimit=false
   )
   &AddVelocity(Velocity=&MakeVector(X=0, Y=-400, Z=0))
  ]
  particleUpdateScript [
   &ParticleState()
   &ScaleColor(
    Scale Alpha=&FloatFromCurve(FloatCurve=DataInterface() {propertyValues {Curve {keys [
         {interpMode=RCIM_Cubic, tangentMode=RCTM_Break, time=0, value=0}
         <at 0.15 1 RCIM_Cubic>
         {interpMode=RCIM_Cubic, tangentMode=RCTM_Break, time=0.8, value=0}
        ]}}})
    ScaleA=true
    ScaleRGB=true
   )
   &ScaleRibbonWidth(Ribbon Width Scale=&FloatFromCurve(FloatCurve=DataInterface() {propertyValues {Curve {keys [{interpMode=RCIM_Cubic, tangentMode=RCTM_Break, time=0, value=1}, {interpMode=RCIM_Cubic, tangentMode=RCTM_Break, time=1, value=0.2}]}}}))
   &Drag(Drag=3, Use Linear Drag=true, Use Rotational Drag=true)
   &GravityForce(Gravity=<-500 0 0>)
   &SolveForcesAndVelocity(Clamp Velocity=false, Limit Acceleration=false)
  ]
  rendererClasses [&NiagaraRibbonRendererProperties]
  renderers [{rendererClass=&NiagaraRibbonRendererProperties}]
 }
 Trail_Front {
  emitterUpdateScript [
   &EmitterState(Life Cycle Mode=::NewEnumerator1, Loop Behavior=::NewEnumerator1, Loop Delay=0.389, Loop Duration Mode=::NewEnumerator1, UseLoopDelay=true)
   &SpawnRate(SpawnRate=50, Use Spawn Probability=false)
   &SpawnRate(
    SpawnRate=&FN_FrontEndFloatModifier(
     Gameplay=&MakeCustomFloatFromBool(
      Bool=&SetBoolByFloatComparison(A=&VectorLength(Vector=$User.OwnerLocalVelocity), B=10)
      True Float=50
      False Float=0
     )
     Front End=60
    )
    Use Spawn Probability=false
   ) {module=SpawnRate001}
  ]
  particleSpawnScript [
   &InitializeParticle(
    Color=<rgba 1 0 0 0.5>
    Lifetime=0.2
    Ribbon Width=20
    Ribbon Width Mode=::NewEnumerator1
    UsePositionOffset=false
    Write Material Random=true
   )
   &SocketLocation(
    SamplePosOffset=true
    Sampled Position Offset=<1 18 0>
    Socket Index=2
    Socket Reader=$System.NewsVan Socket Reader
    Socket Sampling Mode=::NewEnumerator3
    VelocitySpeedLimit=false
   )
   &AddVelocity(Velocity=&MakeVector(X=0, Y=-400, Z=0))
  ]
  particleUpdateScript [
   &ParticleState()
   &ScaleColor(
    Scale Alpha=&FloatFromCurve(FloatCurve=DataInterface() {propertyValues {Curve {keys [
         {interpMode=RCIM_Cubic, tangentMode=RCTM_Break, time=0, value=0}
         <at 0.15 1 RCIM_Cubic>
         {interpMode=RCIM_Cubic, tangentMode=RCTM_Break, time=0.8, value=0}
        ]}}})
    ScaleA=true
    ScaleRGB=true
   )
   &ScaleRibbonWidth(Ribbon Width Scale=&FloatFromCurve(FloatCurve=DataInterface() {propertyValues {Curve {keys [{interpMode=RCIM_Cubic, tangentMode=RCTM_Break, time=0, value=1}, {interpMode=RCIM_Cubic, tangentMode=RCTM_Break, time=1, value=0.2}]}}}))
   &Drag(Drag=3, Use Linear Drag=true, Use Rotational Drag=true)
   &GravityForce(Gravity=<-500 0 0>)
   &SolveForcesAndVelocity(Clamp Velocity=false, Limit Acceleration=false)
  ]
  rendererClasses [&NiagaraRibbonRendererProperties]
  renderers [{rendererClass=&NiagaraRibbonRendererProperties}]
 }
}
summary {
 systemName=NS_KartRocket_Driving_Breakout_FE
 emitters [
  {emitterName=Smoke, rendererClasses [&NiagaraSpriteRendererProperties]}
  {emitterName=Debris, rendererClasses [&NiagaraSpriteRendererProperties]}
  {emitterName=Trail_Front, rendererClasses [&NiagaraRibbonRendererProperties]}
  {emitterName=Trail_Back, rendererClasses [&NiagaraRibbonRendererProperties]}
 ]
 userVariables [
  User.bIsFrontEnd {type=NiagaraBool}
  User.bIsFrontEndPreview {type=NiagaraBool}
  User.OwnerLocalVelocity {type=Vector3f, description="", defaultValue {struct=Vector3f, value=<0 0 0>}}
 ]
}
