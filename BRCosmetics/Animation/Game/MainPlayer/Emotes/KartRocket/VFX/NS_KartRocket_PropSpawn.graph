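# Niagara VFX for the prop-spawn moment of the KartRocket emote, combining a particle burst with short red ribbon trails.
# Burst merges several layers into one emitter, using spawn-group masks for per-layer looks and motion and a visibility tag across four sprite renderers.
# Trail_Front and Trail_Back emit delayed ribbons from sockets read through the system's NewsVan socket reader.
# Standalone Spark, BurstFlame and LWE emitters are disabled; Flame lists no renderers.
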NiagaraSystem /BRCosmetics/Animation/Game/MainPlayer/Emotes/KartRocket/VFX/NS_KartRocket_PropSpawn
dependencies {
 usedDataInterfaces [
  &NiagaraDataInterfaceCurve
  &NiagaraDataInterfaceColorCurve
  &NiagaraDataInterfaceSpriteRendererInfo
  &NiagaraDataInterfaceSocketReader
 ]
 usedDynamicInputs []
 usedModules [
  &self:SystemSpawnScript.SystemScriptSource.SystemScriptGraph.NiagaraNodeAssignment_0.SetVariables_#1
  &SystemState
 ]
}
emitters {
 AirglowLWE {bEnabled=false, rendererClasses [], renderers []}
 Burst {
  emitterUpdateScript [
   &EmitterState()
   &SpawnBurst_Instantaneous(Spawn Count=7, Use Spawn Probability=false, UseLoopCountLimit=false)
   &SpawnBurst_Instantaneous(Spawn Group=1, Use Spawn Probability=false, UseLoopCountLimit=false) {module=SpawnBurst_Instantaneous001}
   &SpawnBurst_Instantaneous(Spawn Count=15, Spawn Group=2, Use Spawn Probability=false, UseLoopCountLimit=false) {module=SpawnBurst_Instantaneous002}
   &SpawnBurst_Instantaneous(Spawn Count=100, Spawn Group=3, Use Spawn Probability=false, UseLoopCountLimit=false) {module=SpawnBurst_Instantaneous003}
   &SpawnBurst_Instantaneous(Spawn Count=20, Spawn Group=4, Use Spawn Probability=false, UseLoopCountLimit=false) {module=SpawnBurst_Instantaneous004}
  ]
  particleSpawnScript [
   /SetVariables(Particles.VisibilityTag=&SpawnGroup()) {isSetParameters=true}
   &InitializeParticle(
    Color=&MaskLinearColorBySpawnGroup(
     Masked Linear Color=<rgba 0.525 0.525 0.525 1>
     Passthrough Linear Color=&MaskLinearColorBySpawnGroup(
      Masked Linear Color=<rgba 1 1 1 1>
      Passthrough Linear Color=&MaskLinearColorBySpawnGroup(
       Masked Linear Color=<rgba 5 3.7043 1.51 1>
       Passthrough Linear Color=&MaskLinearColorBySpawnGroup(Masked Linear Color=<rgba 1 1 1 1>, Passthrough Linear Color=<rgba 1 1 1 1>, Spawn Group Mask=3)
       Spawn Group Mask=2
      )
      Spawn Group Mask=1
     )
    )
    Lifetime=&MaskFloatBySpawnGroup(
     Masked Float=&RandomRangeFloat(Minimum=0.4, Maximum=0.7)
     Passthrough Float=&MaskFloatBySpawnGroup(
      Masked Float=0.2
      Passthrough Float=&MaskFloatBySpawnGroup(
       Masked Float=&RandomRangeFloat(Minimum=0.3, Maximum=1)
       Passthrough Float=&MaskFloatBySpawnGroup(
        Masked Float=&RandomRangeFloat(Minimum=0.3, Maximum=0.7)
        Passthrough Float=&RandomRangeFloat(Minimum=0.3, Maximum=0.6)
        Spawn Group Mask=3
       )
       Spawn Group Mask=2
      )
      Spawn Group Mask=1
     )
    )
    Sprite Rotation Mode=::NewEnumerator2
    Sprite Size Mode=::NewEnumerator3
    Uniform Sprite Size=&MaskFloatBySpawnGroup(
     Masked Float=&RandomRangeFloat(Minimum=42, Maximum=50)
     Passthrough Float=&MaskFloatBySpawnGroup(Masked Float=130, Passthrough Float=&RandomRangeFloat(Minimum=8.5, Maximum=29.75), Spawn Group Mask=1)
    )
    UsePositionOffset=false
    Write Material Random=true
   )
   &ShapeLocation(
    MaskSpawnGroup=false
    Override Seed=false
    Sphere Radius=&MaskFloatBySpawnGroup(
     Masked Float=12
     Passthrough Float=&MaskFloatBySpawnGroup(
      Masked Float=0
      Passthrough Float=&MaskFloatBySpawnGroup(
       Masked Float=10
       Passthrough Float=&MaskFloatBySpawnGroup(Masked Float=10, Passthrough Float=15, Spawn Group Mask=3)
       Spawn Group Mask=2
      )
      Spawn Group Mask=1
     )
    )
    UseOwnerScale=true
   )
   &AddVelocity(
    Module.Use Falloff Distance=false
    Override Seed=false
    Velocity Mode=::NewEnumerator1
    Velocity Speed=&MaskFloatBySpawnGroup(
     Masked Float=&RandomRangeFloat(Minimum=40, Maximum=70)
     Passthrough Float=&MaskFloatBySpawnGroup(
      Masked Float=0
      Passthrough Float=&MaskFloatBySpawnGroup(
       Masked Float=&RandomRangeFloat(Minimum=-240, Maximum=240)
       Passthrough Float=&MaskFloatBySpawnGroup(
        Masked Float=&RandomRangeFloat(Minimum=200, Maximum=400)
        Passthrough Float=&RandomRangeFloat(Minimum=100, Maximum=200)
        Spawn Group Mask=3
       )
       Spawn Group Mask=2
      )
      Spawn Group Mask=1
     )
    )
   )
   &ApplyInitialForces()
   &CameraOffset(Camera Offset Amount=&RandomRangeFloat(Minimum=3, Maximum=10))
  ]
  particleUpdateScript [
   &ParticleState()
   /SetVariables(Particles.SoftFlareDyna:Vector4f=<2 20 0 0>) {isSetParameters=true}
   /SetVariables(Particles.Spark_02_Color:LinearColor=&ColorFromCurve(ColorCurve=DataInterface() {propertyValues {
       AlphaCurve {keys [<at 0.5 4 RCIM_Cubic>,<at 1 0.1 RCIM_Cubic>]}
       BlueCurve {keys [<at 0 0.031 RCIM_Cubic>,<at 1 1 RCIM_Cubic>]}
       GreenCurve {keys [<at 0 0.18 RCIM_Cubic>,<at 1 1 RCIM_Cubic>]}
       RedCurve {keys [<at 0 3 RCIM_Cubic>,<at 1 1 RCIM_Cubic>]}
      }})) {isSetParameters=true}
   /SetVariables(Particles.Spark_02_Size:Vector2f=&Multiply_Vector2D_ByFloat(
     Float=&FloatFromCurve(FloatCurve=DataInterface() {propertyValues {Curve {keys [
          {interpMode=RCIM_Cubic, leaveTangent=8.5101, tangentMode=RCTM_Break, time=0, value=0}
          {interpMode=RCIM_Cubic, tangentMode=RCTM_Break, time=0.2, value=1.2}
          {interpMode=RCIM_Cubic, tangentMode=RCTM_Break, time=1, value=0}
         ]}}})
     Vector2D=&RandomRangeVector2D(Minimum=<0.8 6>, Maximum=<1 8>)
    )) {isSetParameters=true}
   &SpriteRotationRate()
   &ScaleSpriteSize(
    Scale Sprite Size Mode=::NewEnumerator0
    Uniform Scale Factor=&MaskFloatBySpawnGroup(
     Masked Float=&FloatFromCurve(FloatCurve=DataInterface() {propertyValues {Curve {keys [
          {interpMode=RCIM_Cubic, leaveTangent=5.7329, tangentMode=RCTM_Break, time=0, value=0.2}
          {interpMode=RCIM_Cubic, tangentMode=RCTM_Break, time=1, value=3.45}
         ]}}})
     Passthrough Float=&MaskFloatBySpawnGroup(
      Masked Float=&FloatFromCurve(FloatCurve=DataInterface() {propertyValues {Curve {keys [
           {interpMode=RCIM_Cubic, leaveTangent=6.6051, tangentMode=RCTM_Break, time=0, value=0}
           {interpMode=RCIM_Cubic, tangentMode=RCTM_Break, time=0.2, value=1}
           {interpMode=RCIM_Cubic, tangentMode=RCTM_Break, time=1, value=0}
          ]}}})
      Passthrough Float=&FloatFromCurve(FloatCurve=DataInterface() {propertyValues {Curve {keys [
           {interpMode=RCIM_Cubic, leaveTangent=6.6051, tangentMode=RCTM_Break, time=0, value=0}
           {interpMode=RCIM_Cubic, tangentMode=RCTM_Break, time=0.2, value=1}
           {interpMode=RCIM_Cubic, tangentMode=RCTM_Break, time=1, value=0}
          ]}}})
      Spawn Group Mask=1
     )
    )
   )
   &Drag(
    Drag=&MaskFloatBySpawnGroup(
     Masked Float=&RandomRangeFloat(Minimum=2.4, Maximum=3)
     Passthrough Float=&MaskFloatBySpawnGroup(
      Masked Float=0
      Passthrough Float=&MaskFloatBySpawnGroup(
       Masked Float=&RandomRangeFloat(Minimum=3, Maximum=5)
       Passthrough Float=&RandomRangeFloat(Minimum=2, Maximum=3)
       Spawn Group Mask=2
      )
      Spawn Group Mask=1
     )
    )
    Use Linear Drag=true
    Use Rotational Drag=true
   )
   &GravityForce(Gravity=&MaskVectorBySpawnGroup(
     Masked Vector=<0 0 0>
     Passthrough Vector=&MaskVectorBySpawnGroup(
      Masked Vector=<0 0 0>
      Passthrough Vector=&MaskVectorBySpawnGroup(
       Masked Vector=&RandomRangeVector(Minimum=<0 0 120>, Maximum=<0 0 200>)
       Passthrough Vector=&RandomRangeVector(Minimum=<0 0 90>, Maximum=<0 0 150>)
       Spawn Group Mask=2
      )
      Spawn Group Mask=1
     )
    ))
   &VortexForce(
    UseOriginOffset=false
    Vortex Force Amount=&MaskFloatBySpawnGroup(
     Masked Float=0
     Passthrough Float=&MaskFloatBySpawnGroup(
      Masked Float=0
      Passthrough Float=&MaskFloatBySpawnGroup(
       Masked Float=0
       Passthrough Float=&MaskFloatBySpawnGroup(Masked Float=&RandomRangeFloat(Minimum=-600, Maximum=-1200), Passthrough Float=0, Spawn Group Mask=3)
       Spawn Group Mask=2
      )
      Spawn Group Mask=1
     )
    )
   )
   &SolveForcesAndVelocity(Clamp Velocity=false, Limit Acceleration=false)
   &DynamicMaterialParameters(
    Index 0 Param 1=&FloatFromCurve(FloatCurve=DataInterface() {propertyValues {Curve {keys [{interpMode=RCIM_Cubic, tangentMode=RCTM_Break, time=0, value=0}, {interpMode=RCIM_Cubic, tangentMode=RCTM_Break, time=1, value=0.2}]}}})
    Index 1 Param 1=&FloatFromCurve(FloatCurve=DataInterface() {propertyValues {Curve {keys [{interpMode=RCIM_Cubic, tangentMode=RCTM_Break, time=0, value=0}, {interpMode=RCIM_Cubic, tangentMode=RCTM_Break, time=1, value=1}]}}})
    Index 1 Param 2=&FloatFromCurve(FloatCurve=DataInterface() {propertyValues {Curve {keys [
         {interpMode=RCIM_Cubic, tangentMode=RCTM_Break, time=0, value=0.2}
         {interpMode=RCIM_Cubic, leaveTangent=1.7478, tangentMode=RCTM_Break, time=0.4312, value=1.3846}
         {arriveTangent=0.0922, interpMode=RCIM_Cubic, tangentMode=RCTM_Break, time=1, value=2}
        ]}}})
    Index 1 Param 3=&FloatFromCurve(FloatCurve=DataInterface() {propertyValues {Curve {keys [{interpMode=RCIM_Cubic, tangentMode=RCTM_Break, time=0, value=0.2}, {interpMode=RCIM_Cubic, tangentMode=RCTM_Break, time=1, value=1}]}}})
    Index 1 Param 4=&FloatFromCurve(FloatCurve=DataInterface() {propertyValues {Curve {keys [{interpMode=RCIM_Cubic, tangentMode=RCTM_Break, time=0, value=1}, {interpMode=RCIM_Cubic, tangentMode=RCTM_Break, time=1, value=0.8}]}}})
    Index 2 Param 1=&FloatFromCurve(FloatCurve=DataInterface() {propertyValues {Curve {keys [
         {interpMode=RCIM_Cubic, tangentMode=RCTM_Break, time=0, value=0}
         {interpMode=RCIM_Cubic, leaveTangent=1.0205, tangentMode=RCTM_Break, time=0.333, value=0.7}
         {interpMode=RCIM_Cubic, tangentMode=RCTM_Break, time=1, value=1}
        ]}}})
    Index 2 Param 2=&RandomRangeFloat(Minimum=0, Maximum=1)
    Index 2 Param 3=&FloatFromCurve(FloatCurve=DataInterface() {propertyValues {Curve {keys [{interpMode=RCIM_Cubic, tangentMode=RCTM_Break, time=0, value=1}, {interpMode=RCIM_Cubic, tangentMode=RCTM_Break, time=1, value=2}]}}})
    Index 2 Param 4=0.75
    Module.Index0Float0Write=true
    Module.Index0Float1Write=true
    Module.Index0Float2Write=true
    Module.Index0Float3Write=true
    Module.Index1Float0Write=true
    Module.Index1Float1Write=true
    Module.Index1Float2Write=true
    Module.Index1Float3Write=true
    Module.Index2Float0Write=true
    Module.Index2Float1Write=true
    Module.Index2Float2Write=true
    Module.Index2Float3Write=true
    Write Parameter Index 1=::NewEnumerator1
    Write Parameter Index 2=::NewEnumerator1
   )
   &ScaleColor(
    Scale Alpha=&MaskFloatBySpawnGroup(
     Masked Float=&FloatFromCurve(FloatCurve=DataInterface() {propertyValues {Curve {keys [
          {interpMode=RCIM_Cubic, tangentMode=RCTM_Break, time=0, value=0}
          {interpMode=RCIM_Cubic, tangentMode=RCTM_Break, time=0.2, value=1}
          {interpMode=RCIM_Cubic, tangentMode=RCTM_Break, time=0.75, value=1}
          {interpMode=RCIM_Cubic, tangentMode=RCTM_Break, time=1, value=0}
         ]}}})
     Passthrough Float=&MaskFloatBySpawnGroup(
      Masked Float=&FloatFromCurve(FloatCurve=DataInterface() {propertyValues {Curve {keys [
           {interpMode=RCIM_Cubic, leaveTangent=6.6485, tangentMode=RCTM_Break, time=0, value=0}
           {interpMode=RCIM_Cubic, tangentMode=RCTM_Break, time=0.2, value=1}
           {interpMode=RCIM_Cubic, tangentMode=RCTM_Break, time=1, value=0}
          ]}}})
      Passthrough Float=&FloatFromCurve(FloatCurve=DataInterface() {propertyValues {Curve {keys [
           {interpMode=RCIM_Cubic, leaveTangent=6.6485, tangentMode=RCTM_Break, time=0, value=0}
           {interpMode=RCIM_Cubic, tangentMode=RCTM_Break, time=0.2, value=1}
           {interpMode=RCIM_Cubic, tangentMode=RCTM_Break, time=1, value=0}
          ]}}})
      Spawn Group Mask=1
     )
    )
    ScaleA=true
    ScaleRGB=true
   )
  ]
  rendererClasses [&NiagaraSpriteRendererProperties]
  renderers [
   {rendererClass=&NiagaraSpriteRendererProperties}
   {rendererClass=&NiagaraSpriteRendererProperties, rendererIndex=1}
   {rendererClass=&NiagaraSpriteRendererProperties, rendererIndex=2}
   {rendererClass=&NiagaraSpriteRendererProperties, rendererIndex=3}
  ]
 }
 BurstFlame {
  bEnabled=false
  emitterUpdateScript [
   &EmitterState(Life Cycle Mode=::NewEnumerator1, Loop Behavior=::NewEnumerator1, Loop Delay=1, Loop Duration=1.5, UseLoopDelay=false)
   &SpawnBurst_Instantaneous(Spawn Count=20, Use Spawn Probability=false, UseLoopCountLimit=false)
  ]
  particleSpawnScript [
   &InitializeParticle(
    Lifetime Max=0.6
    Lifetime Min=0.3
    Lifetime Mode=::NewEnumerator1
    Sprite Rotation Mode=::NewEnumerator2
    Sprite Size Mode=::NewEnumerator1
    Uniform Sprite Size Max=24
    Uniform Sprite Size Min=7
    UsePositionOffset=false
    Write Material Random=true
   )
   &ShapeLocation(MaskSpawnGroup=false, Override Seed=false, Sphere Radius=15, UseOwnerScale=true)
   &AddVelocity(
    Module.Use Falloff Distance=false
    Override Seed=false
    Velocity Mode=::NewEnumerator1
    Velocity Speed=&RandomRangeFloat(Minimum=100, Maximum=200)
   )
   &DynamicMaterialParameters(
    Index 0 Param 4=2.5
    Module.Param0WriteEnabled=true
    Module.Param1WriteEnabled=true
    Module.Param2WriteEnabled=false
    Module.Param3WriteEnabled=true
   ) {module=DynamicMaterialParameters001}
  ]
  particleUpdateScript [
   /SetVariables(Particles.Size Scale:NiagaraFloat=&FloatFromCurve(
     FloatCurve=DataInterface() {propertyValues {Curve {keys [
         {interpMode=RCIM_Cubic, tangentMode=RCTM_User, time=0, value=0}
         {interpMode=RCIM_Cubic, tangentMode=RCTM_User, time=0.2, value=0.3}
         {tangentMode=RCTM_User, time=1, value=0.1}
        ]}}}
     Scale Curve=6
    )) {isSetParameters=true}
   &Drag(Drag=&RandomRangeFloat(Minimum=1.5, Maximum=3), Use Linear Drag=true, Use Rotational Drag=true)
   &GravityForce(Gravity=&RandomRangeVector(Minimum=<0 0 100>, Maximum=<0 0 200>))
   &SolveForcesAndVelocity(Clamp Velocity=false, Limit Acceleration=false)
   &ParticleState()
   &SpriteRotationRate(Rotation Rate=&RandomRangeFloat(Minimum=-300, Maximum=300)) {module=SpriteRotationRate001}
   &ScaleSpriteSize(Scale Sprite Size Mode=::NewEnumerator0, Uniform Scale Factor=$Particles.Size Scale) {module=ScaleSpriteSize001}
   &ScaleColor(
    Scale Alpha=&Multiply_Float(
     A=4
     B=&FloatFromCurve(FloatCurve=DataInterface() {propertyValues {Curve {keys [<at 0 1 RCIM_Cubic>, {interpMode=RCIM_Cubic, tangentMode=RCTM_User, time=0.5, value=1}, {tangentMode=RCTM_User, time=1, value=0.1}]}}})
    )
    Scale RGB=&Multiply_VectorByFloat(
     Vector=&MakeVectorFromLinearColorRGB(Color=&ColorFromCurve(ColorCurve=DataInterface() {propertyValues {
         AlphaCurve {keys []}
         BlueCurve {keys [<at 0 0.511 RCIM_Cubic>,<at 0.477 0.0313 RCIM_Cubic>]}
         GreenCurve {keys [<at 0 1.2295 RCIM_Cubic>,<at 0.477 0.1797 RCIM_Cubic>]}
         RedCurve {keys [<at 0 6 RCIM_Cubic>,<at 0.477 3 RCIM_Cubic>]}
        }}))
     Float=1
    )
    ScaleA=true
    ScaleRGB=true
   )
   &SubUVAnimation(
    End Frame Range Override=12
    "Sprite Renderer "=DataInterface() {propertyValues {SpriteRenderer=&self:BurstFlame_3.NiagaraSpriteRendererProperties_0}}
    Start Frame Range Override=8
    SubUV Lookup Index Scale:NiagaraFloat=&RandomRangeFloat(Minimum=0.2, Maximum=1)
    UseEndFrame=true
    UseStartFrame=true
   )
  ]
  rendererClasses [&NiagaraSpriteRendererProperties]
  renderers [{rendererClass=&NiagaraSpriteRendererProperties}]
 }
 Flame {rendererClasses [], renderers []}
 SmokeLWE {bEnabled=false, rendererClasses [], renderers []}
 Spark {
  bEnabled=false
  emitterUpdateScript [
   &EmitterState()
   &SpawnBurst_Instantaneous(Spawn Count=100, Use Spawn Probability=false, UseLoopCountLimit=false)
  ]
  particleSpawnScript [
   &InitializeParticle(
    Color=<rgba 2 0.8 0.32 1>
    Lifetime Max=0.7
    Lifetime Min=0.3
    Lifetime Mode=::NewEnumerator1
    Sprite Size Max=<1 8>
    Sprite Size Min=<0.8 6>
    Sprite Size Mode=::NewEnumerator2
    UsePositionOffset=false
    Write Material Random=true
   )
   &ShapeLocation(MaskSpawnGroup=false, Override Seed=false, Sphere Radius=10, UseOwnerScale=true)
   &AddVelocity(
    Module.Use Falloff Distance=false
    Override Seed=false
    Velocity Mode=::NewEnumerator1
    Velocity Speed=&RandomRangeFloat(Minimum=200, Maximum=400)
   )
   &ApplyInitialForces()
  ]
  particleUpdateScript [
   &ParticleState()
   &ScaleSpriteSize(Uniform Curve Sprite Scale=DataInterface() {propertyValues {Curve {keys [
        {interpMode=RCIM_Cubic, leaveTangent=7.1608, tangentMode=RCTM_Break, time=0, value=0}
        {interpMode=RCIM_Cubic, tangentMode=RCTM_Break, time=0.2, value=1.2}
        {interpMode=RCIM_Cubic, tangentMode=RCTM_Break, time=1, value=0}
       ]}}})
   &GravityForce(Gravity=&RandomRangeVector(Minimum=<0 0 90>, Maximum=<0 0 150>))
   &Drag(Drag=&RandomRangeFloat(Minimum=2, Maximum=3), Use Linear Drag=true, Use Rotational Drag=true)
   &VortexForce(UseOriginOffset=false, Vortex Force Amount=&RandomRangeFloat(Minimum=-600, Maximum=-1200))
   &SolveForcesAndVelocity(Clamp Velocity=false, Limit Acceleration=false)
   &Color(Color=&ColorFromCurve(ColorCurve=DataInterface() {propertyValues {
       AlphaCurve {keys [<at 0.5 4 RCIM_Cubic>,<at 1 0.1 RCIM_Cubic>]}
       BlueCurve {keys [<at 0 0.031 RCIM_Cubic>,<at 1 1 RCIM_Cubic>]}
       GreenCurve {keys [<at 0 0.18 RCIM_Cubic>,<at 1 1 RCIM_Cubic>]}
       RedCurve {keys [<at 0 3 RCIM_Cubic>,<at 1 1 RCIM_Cubic>]}
      }}))
  ]
  rendererClasses [&NiagaraSpriteRendererProperties]
  renderers [{rendererClass=&NiagaraSpriteRendererProperties}]
 }
 SparklesLWE {bEnabled=false, rendererClasses [], renderers []}
 Trail_Back {
  emitterUpdateScript [
   &EmitterState(Life Cycle Mode=::NewEnumerator1, Loop Behavior=::NewEnumerator1, Loop Delay=0.512, Loop Duration=0.565, UseLoopDelay=true)
   &SpawnRate(SpawnRate=50, Use Spawn Probability=false) {module=SpawnRate001}
  ]
  particleSpawnScript [
   &InitializeParticle(
    Color=<rgba 1 0 0 0.5>
    Lifetime=0.2
    Ribbon Width=20
    Ribbon Width Mode=::NewEnumerator1
    UsePositionOffset=false
    Write Material Random=true
   )
   &SocketLocation(
    SamplePosOffset=true
    Sampled Position Offset=<3 22 0>
    Socket Index=1
    Socket Reader=$System.NewsVan Socket Reader
    Socket Sampling Mode=::NewEnumerator3
    VelocitySpeedLimit=false
   )
   &AddVelocity(Velocity=<-400 0 0>) {module=AddVelocity001}
  ]
  particleUpdateScript [
   &ParticleState()
   &ScaleColor(
    Scale Alpha=&FloatFromCurve(FloatCurve=DataInterface() {propertyValues {Curve {keys [
         {interpMode=RCIM_Cubic, tangentMode=RCTM_Break, time=0, value=0}
         <at 0.15 1 RCIM_Cubic>
         {interpMode=RCIM_Cubic, tangentMode=RCTM_Break, time=0.8, value=0}
        ]}}})
    ScaleA=true
    ScaleRGB=true
   )
   &ScaleRibbonWidth(Ribbon Width Scale=&FloatFromCurve(FloatCurve=DataInterface() {propertyValues {Curve {keys [{interpMode=RCIM_Cubic, tangentMode=RCTM_Break, time=0, value=1}, {interpMode=RCIM_Cubic, tangentMode=RCTM_Break, time=1, value=0.2}]}}}))
   &Drag(Drag=3, Use Linear Drag=true, Use Rotational Drag=true)
   &GravityForce(Gravity=<-500 0 0>)
   &SolveForcesAndVelocity(Clamp Velocity=false, Limit Acceleration=false)
  ]
  rendererClasses [&NiagaraRibbonRendererProperties]
  renderers [{rendererClass=&NiagaraRibbonRendererProperties}]
 }
 Trail_Front {
  emitterUpdateScript [
   &EmitterState(Life Cycle Mode=::NewEnumerator1, Loop Behavior=::NewEnumerator1, Loop Delay=0.543, Loop Duration=0.596, UseLoopDelay=true)
   &SpawnRate(SpawnRate=50, Use Spawn Probability=false)
  ]
  particleSpawnScript [
   &InitializeParticle(
    Color=<rgba 3 0 0 0.5>
    Lifetime=0.15
    Ribbon Width=20
    Ribbon Width Mode=::NewEnumerator1
    UsePositionOffset=false
    Write Material Random=true
   )
   &SocketLocation(
    SamplePosOffset=true
    Sampled Position Offset=<1 18 0>
    Socket Reader=$System.NewsVan Socket Reader
    Socket Sampling Mode=::NewEnumerator3
    VelocitySpeedLimit=false
   )
   &AddVelocity(Velocity=<-400 0 0>) {module=AddVelocity001}
  ]
  particleUpdateScript [
   &ParticleState()
   &ScaleColor(
    Scale Alpha=&FloatFromCurve(FloatCurve=DataInterface() {propertyValues {Curve {keys [
         {interpMode=RCIM_Cubic, tangentMode=RCTM_Break, time=0, value=0}
         <at 0.15 1 RCIM_Cubic>
         {interpMode=RCIM_Cubic, tangentMode=RCTM_Break, time=0.8, value=0}
        ]}}})
    ScaleA=true
    ScaleRGB=true
   )
   &ScaleRibbonWidth(Ribbon Width Scale=&FloatFromCurve(FloatCurve=DataInterface() {propertyValues {Curve {keys [{interpMode=RCIM_Cubic, tangentMode=RCTM_Break, time=0, value=1}, {interpMode=RCIM_Cubic, tangentMode=RCTM_Break, time=1, value=0.2}]}}}))
   &Drag(Drag=3, Use Linear Drag=true, Use Rotational Drag=true)
   &GravityForce(Gravity=<-500 0 0>)
   &SolveForcesAndVelocity(Clamp Velocity=false, Limit Acceleration=false)
  ]
  rendererClasses [&NiagaraRibbonRendererProperties]
  renderers [{rendererClass=&NiagaraRibbonRendererProperties}]
 }
}
summary {
 systemName=NS_KartRocket_PropSpawn
 emitters [
  {bEnabled=false, emitterName=SmokeLWE, rendererClasses []}
  {bEnabled=false, emitterName=AirglowLWE, rendererClasses []}
  {bEnabled=false, emitterName=SparklesLWE, rendererClasses []}
  {bEnabled=false, emitterName=BurstFlame, rendererClasses [&NiagaraSpriteRendererProperties]}
  {bEnabled=false, emitterName=Spark, rendererClasses [&NiagaraSpriteRendererProperties]}
  {emitterName=Trail_Front, rendererClasses [&NiagaraRibbonRendererProperties]}
  {emitterName=Trail_Back, rendererClasses [&NiagaraRibbonRendererProperties]}
  {emitterName=Burst, rendererClasses [&NiagaraSpriteRendererProperties]}
  {emitterName=Flame, rendererClasses []}
 ]
 userVariables [
  User.bIsFrontEnd {type=NiagaraBool}
  User.bIsFrontEndPreview {type=NiagaraBool}
  User.OwnerLocalVelocity {type=Vector3f, description="", defaultValue {struct=Vector3f, value=<0 0 0>}}
 ]
}
